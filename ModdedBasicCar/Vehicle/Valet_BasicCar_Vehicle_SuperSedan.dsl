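; Fortnite Valet modded SuperSedan vehicle Blueprint, adding shift, drift, boost, lighting and color effects to its parent.
; Gear changes trigger feedback, camera shake and upshift FOV kicks.
; Handbrake drifts drive Niagara and apply occupant GameplayEffects, then catch-up boost adds FX, FOV and timed effects.
; Runtime underglow spotlights share a light-function MID, fade with radio playback, flare on honk, and respond to effects quality, offroad mods and random material color.
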(event Custom|HandleGearChange
  (params (NewGear :type Integer))
  ;[1 1 1] Event ----> Handle any gear change fx/force feedback
  (bind _upshift (CallFunction|HandleShifting NewGear))
  ;; HandleFOVDuringShift
  (if _upshift
    (Variables|Camera|SetCameraFOVOffset -5.0)
    (Utilities|FlowControl|Delay 0.2)
    (Variables|Camera|SetCameraFOVOffset 0.0)))

(event EventBeginPlay
  ;[1 1 1] Begin play events ---> Intialize Underglow
  (|CallParentFunction)
  (Class|FortAsyncActionGameStateReady|GameStateReadyAsync
    (:OnReady
      (Utilities|FlowControl|Sequence
        (:then_0
          (if (Variables|Underglow|GetTryUnderglow)
            (Networking|CanExecuteCosmeticEvents
              (:True
                (CallFunction|SetupUnderglowBindings)
                (Utilities|Array|Clear (Variables|Underglow|GetUnderglowLightComponents))
                (Components|Timeline|SetPlayRate (Variables|Valet_BasicCar_Vehicle_SuperSedan|GetUnderglowFadeTL) (Variables|Underglow|GetUnderglowFadeSpeed))
                (Utilities|EffectsQualityCheck 0
                  (:Cinematic
                    _)
                  (:Epic
                    _)
                  (:High
                    _))
                (CallFunction|CreateUnderglowLightComponents)))))
        (:then_1
          (Vehicle|AssignOnCurrentGearChangedDelegate (EventDispatchers|CreateEvent :!event Custom|HandleGearChange)))))))

(event EventOnStartHandbrake
  ;[1 1 1] HandBrake ---> Drifting FX during the drift
  (|CallParentFunction)
  (if (Class|FortVehicleInterface|IsTouchingGround)
    (Niagara|SetNiagaraVariable(Bool) (Variables|Valet|RoadAndExhaust|Niagara|GetNativeComp_RoadAndExhaust) "IsDrifting" true)
    (EventDispatchers|BindEventtoOnStartHandBrakeCatchupForceDelegate (EventDispatchers|CreateEvent :!event Custom|OnStartHandBrakeCatchupForceDelegate_Event))
    (CallFunction|ApplyDriftGE)))

(event Custom|OnStartHandBrakeCatchupForceDelegate_Event
  (bind _driverpawn (Variables|ValetDefaults|GetDriverPawn))
  (bind _returnvalue (Fortnite|GetFortPlayerControllerfromActor _driverpawn))
  ;[1 1 1] Event ---> Out of drift boost FX and Camera Effects
  (Utilities|FlowControl|Sequence
    (:then_0
      (bind _returnvalue_2 (Niagara|SpawnSystemAttached (Variables|Drift|GetFXDriftingBoost) (Variables|Vehicle|GetSkeletalMesh) "root" "0, 0, 0" "0.000000,180.000000,0.000000" "SnapToTarget" true))
      (Niagara|SetNiagaraVariable(Bool) _returnvalue_2 "Mod_OffroadTires" (Vehicle|Mod|Tire|HasTireMod))
      (Niagara|SetNiagaraVariable(Bool) _returnvalue_2 "TireIntactRL" (Variables|FX|NiagaraSettings|GetTire_Popped_RL))
      (Niagara|SetNiagaraVariable(Bool) _returnvalue_2 "TireIntactRR" (Variables|FX|NiagaraSettings|GetTire_Popped_RR))
      (Niagara|SetNiagaraVariable(Int32) _returnvalue_2 "SedanColor" (Variables|Default|GetRandomMatInt_BasicModdedCar)))
    (:then_1
      (Game|Feedback|ClientPlayForceFeedback _returnvalue (Variables|CatchUpBoost|GetFF_CatchUpBoost) "None" false false false)
      (Projectile|AddGameplayCue%28Looping%29Local _returnvalue (Variables|CatchUpBoost|GetGC_CatchUpBoost) (Ability|GameplayCue|MakeGameplayCueParameters 0.0))
      (CallFunction|ApplyBoostGE)
      (Variables|Camera|SetCameraFOVOffset -1e+01)
      (Utilities|FlowControl|Delay 1.0)
      (Projectile|RemoveGameplayCue%28Looping%29Local (Fortnite|GetFortPlayerControllerfromActor (Variables|ValetDefaults|GetDriverPawn)) (Variables|CatchUpBoost|GetGC_CatchUpBoost) (Ability|GameplayCue|MakeGameplayCueParameters 0.0))
      (Variables|Camera|SetCameraFOVOffset 0.0)
      (CallFunction|CleanUpBoostGE))))

(event EventOnPawnEnterVehicle
  (params (PlayerPawn :type Fort%20Player%20Pawn%20Object%20Reference) (SeatIdx :type Integer))
  ;[1 1 1] HandBrake ---> Drifting FX during the drift
  (|CallParentFunction PlayerPawn SeatIdx)
  (Niagara|SetNiagaraVariable(Int32) (Variables|Valet|RoadAndExhaust|Niagara|GetNativeComp_RoadAndExhaust) "SedanColor" (Variables|Default|GetRandomMatInt_BasicModdedCar)) ; Set FX tail light color
)

(event EventOnPawnExitVehicle
  (params (PlayerPawn :type Fort%20Player%20Pawn%20Object%20Reference) (ExitSocketName :type Name :default "None"))
  ;[1 1 1] HandBrake ---> Drifting FX during the drift
  (|CallParentFunction PlayerPawn ExitSocketName)
  (Niagara|SetNiagaraVariable(Bool) (Variables|Valet|RoadAndExhaust|Niagara|GetNativeComp_RoadAndExhaust) "IsDrifting" false))

(event EventPawnExitSeat
  (params (Vehicle :type Fort%20Vehicle%20Interface%20Interface%20%28by%20ref%29) (PlayerPawn :type Fort%20Pawn%20Object%20Reference) (SeatIndex :type Integer))
  ;[1 1 1] HandBrake ---> Drifting FX during the drift
  (Utilities|FlowControl|Sequence
    (:then_0
      (CallFunction|CleanUpDriftGE)
      (CallFunction|CleanUpBoostGE)
      (Projectile|RemoveGameplayCue%28Looping%29Local (Fortnite|GetFortPlayerControllerfromActor (Variables|ValetDefaults|GetDriverPawn)) (Variables|CatchUpBoost|GetGC_CatchUpBoost) (Ability|GameplayCue|MakeGameplayCueParameters 0.0)))
    (:then_1
      (|CallParentFunction Vehicle PlayerPawn SeatIndex))))

(event EventOnStopHandbrake
  ;[1 1 1] HandBrake ---> Drifting FX during the drift
  (|CallParentFunction)
  (CallFunction|CleanUpDriftGE)
  (Niagara|SetNiagaraVariable(Bool) (Variables|Valet|RoadAndExhaust|Niagara|GetNativeComp_RoadAndExhaust) "IsDrifting" false))

(event Custom|SetupUnderglowBindings
  ;[1 1 1] Underglow
  (Utilities|FlowControl|Sequence
    (:then_0
      (EventDispatchers|BindEventtoOnEffectsQualityChanged (Settings|GetFortGameUserSettings) (EventDispatchers|CreateEvent :!event Handle%20on%20Effects%20Quality%20Changed))
      (Vehicle|Mods|BindEventtoOnModAppliedonClientDelegate (EventDispatchers|CreateEvent :!event UpdateUnderglowForOffroad) :!class /Script/FortniteGame.FortAthenaVehicle))
    (:then_1
      (bind _output_get (Variables|Default|SetStreamingRadio (Actor|GetComponentbyClass "/Script/SrirachaRanch.StreamingRadioPlayerComponent")))
      (Utilities|IsValid _output_get
        (:Is%20Valid
          (Radio|AssignOnReplicatedStateChangedDelegate (Variables|Default|GetStreamingRadio) (EventDispatchers|CreateEvent :!event On%20Radio%20State%20Changed)))))))

(event EventHandleOnExplode
  (params (LastDamageInstigator :type Controller%20Object%20Reference) (Vehicle :type Actor%20Object%20Reference))
  ;[1 1 1] Event ----> Handle Lens Flare on explosion and headlight change
  (|CallParentFunction LastDamageInstigator Vehicle)
  (Components|Activation|Deactivate (Variables|Default|GetNS_SuperSedan_LensFlare)))

(event OnLightEnabledChanged(ValetLightSpawnerComponent)
  (params (enabled :type Boolean))
  (bind _ns_supersedan_lensflare (Variables|Default|GetNS_SuperSedan_LensFlare))
  ;[1 1 1] Event ----> Handle Lens Flare on explosion and headlight change
  (Components|Activation|SetActive _ns_supersedan_lensflare enabled true)
  (Rendering|SetVisibility _ns_supersedan_lensflare enabled false) ; There is a slight delay when deactivating that looks weird so toggling vis\n
)

(event Custom|UnderglowFadeInOut
  (params (In/Out :type Boolean))
  ;[1 1 1] Underglow
  (if (Variables|Underglow|GetIsUnderglowAvailable)
    (if In/Out
      (bind (_direction _0_1) (Timeline|UnderglowFadeTL Play)
        (:Update
          (CallFunction|SetNormalizedUnderglowIntensity _0_1)))
      (else
        (Timeline|UnderglowFadeTL Reverse)))))

(event Vehicle|EventOnHonk
  ;[1 1 1] Underglow
  (Networking|CanExecuteCosmeticEvents
    (:True
      (if (and (Variables|Underglow|GetIsUnderglowAvailable) (Variables|Underglow|GetIsRadioOn))
        (CallFunction|BoostUnderglow)))
    (:False
      _))
  (|CallParentFunction))

(event EventPickRandomMaterial
  ;[1 1 1] Initialize Underglow Color based on Materials
  (|CallParentFunction)
  (Rendering|Components|Light|SetLightColor (Variables|Underglow|GetUnderglowLightComponents) (Utilities|Array|Get(acopy) (Variables|Underglow|GetUnderglowColor_Randomize) (Variables|Default|GetRandomMatInt_BasicModdedCar)) true)
  ;[1 1 1] Initialize Headlight & Tail FX Color based on Materials
  (bind _result (Math|Integer|IncrementInt (Variables|Lights|GetUpdateValetHeadlightColors))
    (:%20%20
      (Variables|Lights|SetUpdateValetHeadlightColors _result)
      (Niagara|SetNiagaraVariable(Int32) (Variables|Default|GetNS_SuperSedan_LensFlare) "SedanColor" (Variables|Default|GetRandomMatInt_BasicModdedCar)) ; Set FX tail light color
    )))

(fn ApplyBoostGE
  (GameplayEffects|ApplyGameplayEffectToSelf :self (Ability|GetAbilitySystemComponent (Variables|ValetDefaults|GetDriverPawn)) :GameplayEffectClass (Variables|CatchUpBoost|GetBoostTag) :Level 0.0)
  (for _array_element (Vehicle|GetAllPassengers)
    (Utilities|IsValid _array_element
      (:Is%20Valid
        (GameplayEffects|ApplyGameplayEffectToSelf :self (Ability|GetAbilitySystemComponent _array_element) :GameplayEffectClass (Variables|CatchUpBoost|GetBoostTag) :Level 0.0))))
  (return))

(fn ApplyDriftGE
  (GameplayEffects|ApplyGameplayEffectToSelf :self (Ability|GetAbilitySystemComponent (Variables|ValetDefaults|GetDriverPawn)) :GameplayEffectClass (Variables|Drift|GetDriftingTag) :Level 0.0)
  (for _array_element (Vehicle|GetAllPassengers)
    (Utilities|IsValid _array_element
      (:Is%20Valid
        (GameplayEffects|ApplyGameplayEffectToSelf :self (Ability|GetAbilitySystemComponent _array_element) :GameplayEffectClass (Variables|Drift|GetDriftingTag) :Level 0.0))))
  (return))

(fn BoostUnderglow
  (Utilities|Time|SetTimerbyEvent (EventDispatchers|CreateEvent :!event Deactivate%20Underglow%20Boost) (Variables|Underglow|GetUnderglowBoostDuration))
  (Rendering|Material|SetScalarParameterValue (Variables|Underglow|GetMID_Underglow) "IsReactivityActive" 1.0 :!class /Script/Engine.MaterialInstanceDynamic)
  (CallFunction|SetNormalizedUnderglowIntensity (Variables|Underglow|GetUnderglowBoostFactor)))

(fn CleanupBoostGE
  (GameplayEffects|RemoveActiveGameplayEffectbySourceEffect (Ability|GetAbilitySystemComponent (Variables|ValetDefaults|GetDriverPawn)) (Variables|CatchUpBoost|GetBoostTag) 0 -1)
  (for _array_element (Vehicle|GetAllPassengers)
    (Utilities|IsValid _array_element
      (:Is%20Valid
        (GameplayEffects|RemoveActiveGameplayEffectbySourceEffect (Ability|GetAbilitySystemComponent _array_element) (Variables|CatchUpBoost|GetBoostTag) 0 -1))))
  (return))

(fn CleanupDriftGE
  (GameplayEffects|RemoveActiveGameplayEffectbySourceEffect (Ability|GetAbilitySystemComponent (Variables|ValetDefaults|GetDriverPawn)) (Variables|Drift|GetDriftingTag) 0 -1)
  (for _array_element (Vehicle|GetAllPassengers)
    (Utilities|IsValid _array_element
      (:Is%20Valid
        (GameplayEffects|RemoveActiveGameplayEffectbySourceEffect (Ability|GetAbilitySystemComponent _array_element) (Variables|Drift|GetDriftingTag) 0 -1))))
  (return))

(fn CreateUnderglowLightComponents
  (Utilities|FlowControl|Sequence
    (:then_0
      (Utilities|IsValid (Variables|Underglow|GetMID_Underglow)
        (:Is%20Not%20Valid
          (bind _returnvalue (Rendering|Material|CreateDynamicMaterialInstance "/Valet/Materials/M_LF_Underglow.M_LF_Underglow" "None" "None" :!class /Script/Engine.KismetMaterialLibrary))
          (Variables|Underglow|SetMID_Underglow _returnvalue))))
    (:then_1
      (for _index (range 4)
        (bind _returnvalue_2 (AddComponent|NODE_AddSpotLightComponent-0 :RelativeTransform (Utilities|Array|Get(acopy) (Variables|Underglow|GetUnderglowTransforms) _index)))
        (Utilities|Array|Add (Variables|Underglow|GetUnderglowLightComponents) _returnvalue_2)
        (Rendering|Components|Light|SetLightFunctionMaterial _returnvalue_2 (Variables|Underglow|GetMID_Underglow))
        (Rendering|Components|Light|SetLightColor (Variables|Underglow|GetUnderglowLightComponents) (Utilities|Array|Get(acopy) (Variables|Underglow|GetUnderglowColor_Randomize) (Variables|Default|GetRandomMatInt_BasicModdedCar)) true)))
    (:then_2
      (Variables|Underglow|SetIsUnderglowAvailable true)
      (CallFunction|InitializeUnderglow))))

(fn DeactivateUnderglowBoost
  (if (Variables|Underglow|GetIsRadioOn)
    (Rendering|Material|SetScalarParameterValue (Variables|Underglow|GetMID_Underglow) "IsReactivityActive" 0.0 :!class /Script/Engine.MaterialInstanceDynamic)
    (CallFunction|SetNormalizedUnderglowIntensity 1.0)))

(fn DestroyUnderglowLightComponents
  (for _array_element (Variables|Underglow|GetUnderglowLightComponents)
    (Components|DestroyComponent _array_element))
  (Utilities|Array|Clear (Variables|Underglow|GetUnderglowLightComponents))
  (Variables|Underglow|SetIsUnderglowAvailable false))

(fn HandleonEffectsQualityChanged
  (bind _isunderglowavailable (Variables|Underglow|GetIsUnderglowAvailable))
  (Utilities|EffectsQualityCheck 0
    (:Cinematic
      _)
    (:Epic
      _)
    (:High
      _)
    (:Medium
      (if _isunderglowavailable :!label _label
        (CallFunction|DestroyUnderglowLightComponents)))
    (:Low
      (goto _label)))
  (if (not _isunderglowavailable)
    (CallFunction|CreateUnderglowLightComponents)))

(fn HandleShifting
  (params (NewGear :type Integer))
  (bind _driverpawn (Variables|ValetDefaults|GetDriverPawn))
  (bind _returnvalue (Fortnite|GetFortPlayerControllerfromActor _driverpawn))
  (bind _driverpawn_2 (Variables|ValetDefaults|GetDriverPawn))
  (bind _returnvalue_2 (Fortnite|GetFortPlayerControllerfromActor _driverpawn_2))
  (if (> NewGear (Variables|ShiftingGears|GetCurrentGearAudio))
    (Development|PrintString "UpShift")
    (Game|Feedback|ClientPlayForceFeedback _returnvalue_2 (Variables|ShiftingGears|GetFF_Upshift) "None" false false false)
    (Game|Feedback|ClientStartCameraShake _returnvalue_2 (Variables|ShiftingGears|GetUpShiftCameraShake) 1.0 "CameraLocal" (Transformation|GetActorRotation _driverpawn_2))
    (Variables|ShiftingGears|SetCurrentGearAudio NewGear)
    (return true)
    (else
      (Development|PrintString "Downshift")
      (Game|Feedback|ClientPlayForceFeedback _returnvalue (Variables|ShiftingGears|GetFF_Downshift) "None" false false false)
      (Game|Feedback|ClientStartCameraShake _returnvalue (Variables|ShiftingGears|GetDownShiftCameraShake) 1.0 "CameraLocal" (Transformation|GetActorRotation))
      (Variables|ShiftingGears|SetCurrentGearAudio NewGear)
      (return false))))

(fn InitializeUnderglow
  (Variables|Underglow|SetUnderglowBaseIntensity (Variables|/Script/Engine.LightComponentBase|GetIntensity (Utilities|Array|Get(acopy) (Variables|Underglow|GetUnderglowLightComponents))))
  (CallFunction|SetNormalizedUnderglowIntensity (Math|Float|SelectFloat 1.0 0.0 (Variables|Underglow|GetIsRadioOn)))
  (if (Variables|Mods|GetIsOffroadModded)
    (CallFunction|UpdateUnderglowforOffroad 0 (Variables|Mods|GetModTag_Offroad))))

(fn OnRadioStateChanged
  (params (NewState :type EStreamingRadioSourceState%20Enum))
  (bind _output_get (Variables|Underglow|SetIsRadioOn (Utilities|Enum|Equal(Enum) NewState "Playing")))
  (CallFunction|UnderglowFadeinOut _output_get))

(fn OnRep_UpdateValetHeadlightColors
  (if (not (Networking|IsServer))
    (ValetLights|SetValetLightColor (Variables|Default|GetValetLightSpawnerComponent) (Utilities|Array|Get(acopy) (Variables|Lights|GetHeadlightColor_Randomizer) (Variables|Default|GetRandomMatInt_BasicModdedCar)))))

(fn SetNormalizedUnderglowIntensity
  (params (Intensity :type Float%20%28double-precision%29))
  (bind _underglowlightcomponents (Variables|Underglow|GetUnderglowLightComponents))
  (Rendering|SetVisibility _underglowlightcomponents (> Intensity 0) false)
  (Rendering|Components|Light|SetIntensity _underglowlightcomponents (* Intensity (Variables|Underglow|GetUnderglowBaseIntensity))))

(fn UpdateUnderglowForOffroad
  (params (Vehicle :type Actor%20Object%20Reference) (ModTag :type Gameplay%20Tag%20Structure%20%28by%20ref%29))
  (bind _underglowlightcomponents (Variables|Underglow|GetUnderglowLightComponents))
  (if (== ModTag (Variables|Mods|GetModTag_Offroad))
    (Variables|Mods|SetIsOffroadModded true)
    (if (Variables|Underglow|GetIsUnderglowAvailable)
      (Rendering|Lighting|SetAttenuationRadius _underglowlightcomponents 1.8e+02)
      (Rendering|Lighting|SetLightFalloffExponent _underglowlightcomponents 4.0)
      (return))))

(fn ConstructionScript
  (|CallParentFunction))
